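# Unlit alpha-composite brush material for the CloudCompositing tool that stamps strength or velocity onto the previous buffer, by default a cloud mask render target.
# An exponential radial gradient at a parameterized position and radius forms the brush, its density perturbed by gradient noise with optional time-based randomization.
# PaintVelocity switches between channel-selected strength and a velocity vector, while IsAddMode chooses additive or interpolated blending.
# RGB is clamped non-negative unless painting velocity, and alpha is output inverted.
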Material /Volumetrics/Tools/CloudCompositing/Blueprints/GameSetup/M_AddForce
graph {
 properties {
  bAllowNegativeEmissiveColor=true
  blendMode=BLEND_AlphaComposite
  pixelDepthOffsetMode=PDOM_Legacy
  shadingModel=MSM_Unlit
  usedShadingModels=Unlit
 }
 nodes [
  n1 = ScalarParameter() {parameterName=PaintVelocity}
  n2 = TextureSampleParameter2D() {parameterName=PreviousBuffer, samplerType=SAMPLERTYPE_LinearColor, texture=/Volumetrics/Tools/CloudCompositing/RenderTargets/RT_CloudMask_01}
  n3 = ScalarParameter() {defaultValue=1, parameterName=Strength}
  n4 = VectorParameter() {parameterName=Velocity, defaultValue=<rgba 1 0 0 0>}
  n5 = AppendVector(A=@n4.RGB, B=@n4.RGB)
  n6 = LinearInterpolate(A=@n3, B=@n5, Alpha=@n1)
  n7 = VectorParameter() {parameterName=P, defaultValue=<rgba 0.5 0.5 0 0>}
  n8 = ScalarParameter() {defaultValue=0.1, parameterName=R}
  n9 = Constant() {r=1}
  n10 = VectorParameter() {parameterName=ColorChannel, defaultValue=<rgba 0 0 0 1>}
  n11 = AppendVector(A=@n10.RGB, B=@n10.RGB)
  n12 = MaterialFunctionCall(Float4=Multiply(
    A=MaterialFunctionCall(CenterPosition=@n7.RGB, Radius=@n8, Density=@n9) {materialFunction=&RadialGradientExponential}.RadialGradientExponential
    B=@n11
   )) {materialFunction=&BreakOutFloat4Components}
  n13 = LinearInterpolate(A=@n2.R, B=@n6, Alpha=@n12.R)
  n14 = LinearInterpolate(A=@n2.G, B=@n6, Alpha=@n12.G)
  n15 = LinearInterpolate(A=@n2.B, B=@n6, Alpha=@n12.B)
  n16 = LinearInterpolate(A=@n2.A, B=@n6, Alpha=@n12.A)
  n17 = ScalarParameter() {defaultValue=1, parameterName=D}
  n18 = Time()
  n19 = AppendVector(
   A=TextureCoordinate()
   B=LinearInterpolate(A=Constant(), B=Multiply(A=@n18) {constB=0.618}, Alpha=ScalarParameter() {parameterName=Randomize})
  )
  n20 = ScalarParameter() {defaultValue=5, parameterName=NoiseTiling}
  n21 = Multiply(
   A=Noise!(World Position=Multiply(A=@n19, B=@n20), FilterWidth) {noiseFunction=NOISEFUNCTION_GradientALU}
   B=ScalarParameter() {defaultValue=1, parameterName=NoiseAmount}
  )
  n22 = Multiply(
   A=MaterialFunctionCall(CenterPosition=@n7.RGB, Radius=@n8, Density=Add(A=@n17, B=@n21)) {materialFunction=&RadialGradientExponential}.RadialGradientExponential
   B=@n3
  )
  n23 = LinearInterpolate(A=@n11, B=@n5, Alpha=@n1)
  n24 = LinearInterpolate(
   A=LinearInterpolate(
    A=MaterialFunctionCall(X=@n13, Y=@n14, Z=@n15, A=@n16) {materialFunction=&MakeFloat4}.Result
    B=LinearInterpolate(A=@n2.RGBA, B=@n5, Alpha=@n22)
    Alpha=@n1
   )
   B=Add(A=@n2.RGBA, B=Multiply(A=@n22, B=@n23))
   Alpha=ScalarParameter() {defaultValue=1, parameterName=IsAddMode, sliderMax=1}
  )
  n25 = ComponentMask(None=@n24) {b=true, g=true, r=true}
  If(A=@n1, B=Constant() {r=1}, "A > B"=@n25, "A < B"=Max(A=@n25) {constB=0})
  OneMinus(None=ComponentMask(None=@n24) {a=true})
 ]
}
