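# Translucent, unlit, two-sided material that previews a brush in the Volumetrics cloud compositing tool, drawn without depth testing.
# Virtual plane projections at the object's height provide a world-anchored grid and a circle sized to object radius.
# The circle and a height-distance sphere mask blend the tinted grid over a background value.
# A preview opacity chain multiplies similar masks by view angle, corrected for camera side, and a near-camera fade.
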Material /Volumetrics/Tools/CloudCompositing/Blueprints/GameSetup/M_BrushDisplay
graph {
 properties {
  bDisableDepthTest=true
  bUseTranslucencyVertexFog=false
  blendMode=BLEND_Translucent
  pixelDepthOffsetMode=PDOM_Legacy
  shadingModel=MSM_Unlit
  twoSided=true
  usedShadingModels=Unlit
 }
 nodes [
  n1 = VectorParameter() {parameterName=E, defaultValue=<rgba 0 1 1 1>}
  n2 = Multiply(A=@n1.RGB, B=@n1.RGB) {constB=0.05}
  n3 = MaterialFunctionCall(
   Plane Normal Axis=Constant3Vector() {constant=<rgba 0 0 1 0>}
   Plane X Axis=Constant3Vector() {constant=<rgba 1 0 0 0>}
   Plane Center=Multiply(A=ObjectPositionWS(), B=Constant3Vector() {constant=<rgba 0 0 1 0>})
   UV Scale=ScalarParameter() {defaultValue=1000000, parameterName=BaseGridSize}
  ) {materialFunction=&VirtualPlaneCoordinates}
  n4 = ScalarParameter() {defaultValue=4, parameterName=Grid Tiling}
  n5 = LinearInterpolate(Alpha=TextureSample(UVs=Multiply(A=@n3.UVs, B=@n4)) {texture=/Engine/Functions/Engine_MaterialFunctions01/Debug/T_Grid}.R) {constA=0.2, constB=2}
  n6 = ScalarParameter() {defaultValue=0.1, parameterName=Grid}
  n7 = Constant() {r=0.5}
  n8 = ObjectRadius()
  n9 = MaterialFunctionCall(
   Plane Normal Axis=Constant3Vector() {constant=<rgba 0 0 1 0>}
   Plane X Axis=Constant3Vector() {constant=<rgba 1 0 0 0>}
   Plane Center=ObjectPositionWS()
   UV Scale=Multiply(A=ComponentMask(None=@n8) {r=true}) {constB=2}
  ) {materialFunction=&VirtualPlaneCoordinates}
  n10 = SphereMask(A=@n9.UVs, B=@n7, Radius=@n7, Hardness=Constant() {r=0.5})
  n11 = SphereMask(
   A=ComponentMask(None=WorldPosition().XYZ) {b=true}
   B=ComponentMask(None=ObjectPositionWS()) {b=true}
   Radius=ScalarParameter() {defaultValue=15000, parameterName=Rad}
   Hardness=ScalarParameter() {defaultValue=0.5, parameterName=Hardness}
  )
  LinearInterpolate(
   A=ScalarParameter() {parameterName=Brush Background}
   B=Multiply(A=Multiply(A=@n2) {constB=0.3}, B=LinearInterpolate(B=@n5, Alpha=@n6) {constA=1})
   Alpha=Saturate(None=Add(A=@n10, B=@n11))
  ) {constA=1}
  n12 = Constant() {r=0.5}
  n13 = Add(A=SphereMask(A=@n9.UVs, B=@n12, Radius=@n12, Hardness=Constant() {r=1}), B=@n11)
  n14 = CameraVectorWS()
  n15 = Subtract(A=CameraPositionWS(), B=ObjectPositionWS())
  n16 = Saturate(None=Ceil(None=ComponentMask(None=@n15) {b=true}))
  n17 = Saturate(None=Multiply(A=ComponentMask(None=@n14) {b=true}, B=LinearInterpolate(Alpha=@n16) {constA=-1}))
  n18 = Subtract(A=Distance(A=ObjectPositionWS(), B=CameraPositionWS()), B=ScalarParameter() {defaultValue=100, parameterName=CameraFadeoutMin})
  Multiply(
   A=Multiply(A=Multiply(A=@n13, B=@n17), B=ScalarParameter() {defaultValue=0.1, parameterName=Brush Preview Opacity})
   B=Saturate(None=Divide(A=@n18) {constB=900})
  )
  Multiply(A=@n2, B=ScalarParameter() {defaultValue=1, parameterName=Opacity})
  Multiply(A=ComponentMask(None=ObjectRadius()) {r=true}) {constB=2}
 ]
}
