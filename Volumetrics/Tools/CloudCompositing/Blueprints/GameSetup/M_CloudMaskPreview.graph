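# Unlit, two-sided additive material that previews the RT_CloudMask_01 cloud mask render target with a border.
# Mask RGB is scaled by Show Preview Plane and added to an edge mask floored from centred UV distance and Border Width.
# The edge Max takes the R component twice, so only U edges register.
# MPC_Clouds SkyTextureScaleKm combines with Border Width in two math nodes that feed nothing.
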Material /Volumetrics/Tools/CloudCompositing/Blueprints/GameSetup/M_CloudMaskPreview
graph {
 properties {blendMode=BLEND_Additive, pixelDepthOffsetMode=PDOM_Legacy, shadingModel=MSM_Unlit, twoSided=true, usedShadingModels=Unlit}
 nodes [
  n1 = TextureCoordinate()
  n2 = ScalarParameter() {defaultValue=0.02, parameterName=Border Width}
  n3 = Multiply(A=Abs(None=Subtract(A=@n1) {constB=0.5}), B=Add(A=@n2) {constB=2}) {constB=1.975}
  n4 = MaterialFunctionCall(Float2=Saturate(None=Floor(None=@n3))) {materialFunction=&BreakOutFloat2Components}
  Add(
   A=Multiply(
    A=TextureSample() {samplerType=SAMPLERTYPE_LinearColor, texture=/Volumetrics/Tools/CloudCompositing/RenderTargets/RT_CloudMask_01}.RGB
    B=ScalarParameter() {parameterName=Show Preview Plane}
   )
   B=Max(A=@n4.R, B=@n4.R)
  )
  n5 = CollectionParameter() {collection=/Volumetrics/Tools/CloudCompositing/MPC/MPC_Clouds, parameterName=SkyTextureScaleKm}
  Divide(A=@n2, B=@n5)
  Multiply(A=@n5, B=@n2)
 ]
}
