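# M_SetForce is an unlit translucent material in the CloudCompositing tool, apparently a brush for painting force or velocity values.
# Strength scales a blend between the ColorChannel and Velocity vector parameters, selected by PaintVelocity, with negative emissive values allowed for signed data.
# An exponential radial gradient provides a circular falloff from center P, radius R and density D, alongside an inverted copy.
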Material /Volumetrics/Tools/CloudCompositing/Blueprints/GameSetup/M_SetForce
graph {
 properties {
  bAllowNegativeEmissiveColor=true
  blendMode=BLEND_Translucent
  pixelDepthOffsetMode=PDOM_Legacy
  shadingModel=MSM_Unlit
  usedShadingModels=Unlit
 }
 nodes [
  Multiply(
   A=ScalarParameter() {defaultValue=1, parameterName=Strength}
   B=LinearInterpolate(
    A=VectorParameter() {parameterName=ColorChannel, defaultValue=<rgba 1 0 0 0>}.RGB
    B=VectorParameter() {parameterName=Velocity, defaultValue=<rgba 1 0 0 0>}.RGB
    Alpha=ScalarParameter() {parameterName=PaintVelocity}
   )
  )
  n1 = ScalarParameter() {defaultValue=0.1, parameterName=R}
  Multiply(A=@n1) {constB=0.333}
  OneMinus(None=MaterialFunctionCall(
    CenterPosition=VectorParameter() {parameterName=P, defaultValue=<rgba 0.5 0.5 0 0>}.RGB
    Radius=@n1
    Density=ScalarParameter() {defaultValue=1, parameterName=D}
   ) {materialFunction=&RadialGradientExponential}.RadialGradientExponential)
 ]
}
