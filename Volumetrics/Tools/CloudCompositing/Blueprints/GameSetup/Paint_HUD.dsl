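; Paint_HUD is a Blueprint in the cloud compositing tool of the Volumetrics plugin that sets up the volumetric painting UI.
; On BeginPlay it creates the UI_VolumetricPainting_Toolbar widget for player controller 0 and adds it to the viewport.
; Tick, overlap and the construction script are empty.
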(event EventBeginPlay
  ;[1 1 1] Create UI Paint Toolbar Widget
  (bind _returnvalue (UserInterface|CreateWidget "/Volumetrics/Tools/CloudCompositing/Blueprints/GameSetup/UI/UI_VolumetricPainting_Toolbar.UI_VolumetricPainting_Toolbar_C" (Game|GetPlayerController 0)))
  (UserInterface|Viewport|AddtoViewport _returnvalue 0))

(event Collision|EventActorBeginOverlap
  (params (OtherActor :type Actor%20Object%20Reference)))

(event EventTick
  (params (DeltaSeconds :type Float%20%28single-precision%29)))

(fn ConstructionScript)
