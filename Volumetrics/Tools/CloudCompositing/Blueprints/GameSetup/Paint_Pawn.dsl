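; Player pawn for painting volumetric clouds, with free-fly movement and mouse-look.
; Each tick it finds the brush point by cloud-elevation plane intersection, line trace, or the nearer of both, per BP_PaintClouds' trace mode, then moves the preview mesh.
; While pressed, it passes location, size and Shift-invertible strength to BP_PaintClouds and draws with the current blend mode.
; BP_PaintClouds registers through InitPaintCloudsBP; key and mouse-button events show no connected logic.
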(event EventBeginPlay
  ;[1 1 1] On Begin Play
  (Viewport|SetViewportMouseCaptureMode "CapturePermanently_IncludingInitialMouseDown")
  (Variables|/Script/Engine.PlayerController|SetbShowMouseCursor true (Game|GetPlayerController 0))
  (bind _outactors (Actor|GetAllActorsOfClass "/Volumetrics/Tools/FluidSimulation/Blueprints/BP_FluidSimulation_01.BP_FluidSimulation_01_C"))
  (Variables|Default|SetFluidSimBP (Utilities|Array|Get(acopy) _outactors))
  (bind _returnvalue (Rendering|Material|CreateDynamicMaterialInstance (Variables|Default|GetBrushPreview) 0 0 "None" :!class /Script/Engine.PrimitiveComponent))
  (Variables|Default|SetBrushPreviewMID _returnvalue)
  (Transformation|SetWorldScale3D (Variables|Default|GetBrushPreview) (Math|Conversions|ToVector(Float) (CallFunction|ConvertBrushSize (Variables|Default|GetBrushSize)))))

(event LeftMouseButton
  (params (Key :type Key%20Structure)))

(event RightMouseButton
  (params (Key :type Key%20Structure)))

(event EventTick
  (params (DeltaSeconds :type Float%20%28single-precision%29))
  (bind _returnvalue (Game|GetPlayerController 0))
  (bind _returnvalue_2 (Pawn|GetControlRotation _returnvalue :!class /Script/Engine.Controller))
  (bind (_x _y _z) (Math|Rotator|GetAxes _returnvalue_2))
  (bind _paintcloudsbp (Variables|Default|GetPaintCloudsBP))
  (CallFunction|RayTraceToCloudLayer)
  (CallFunction|UpdateBrushPreview)
  (if (Variables|Default|GetPressed)
    (Utilities|IsValid _paintcloudsbp
      (:Is%20Valid
        (Variables|/Volumetrics/Tools/CloudCompositing/Blueprints/BP_PaintClouds.BP_PaintClouds_C|SetBrush%20Strength (* (Variables|Default|GetOpacity) (Math|Float|SelectFloat -1.0 1.0 (Variables|Default|GetShift))) _paintcloudsbp)
        (Class|BPPaintClouds|UpdatePaintLocation (Variables|Default|GetPaintCloudsBP) (Variables|Default|GetBrush%20Hit%20Location) (Variables|Default|GetPrevious%20Brush%20Hit%20Location))
        (Variables|/Volumetrics/Tools/CloudCompositing/Blueprints/BP_PaintClouds.BP_PaintClouds_C|SetBrush%20Size (* (Variables|Default|GetBrushSize) 0.1) (Variables|Default|GetPaintCloudsBP))
        (Class|BPPaintClouds|Draw (Variables|Default|GetPaintCloudsBP) (Variables|Default|GetPaintBlendMode))))
    (elif (Variables|Default|GetRight)
      (Pawn|SetControlRotation _returnvalue (Math|Rotator|MakeRotator (.roll _returnvalue_2) (+ (.pitch _returnvalue_2) (* (Input|MouseValues|GetMouseY) 2.0)) (+ (.yaw _returnvalue_2) (* (Input|MouseValues|GetMouseX) 2.0))))))
  (Transformation|AddActorWorldOffset (+ (+ (* _x (* (Math|Conversions|ToFloat(Boolean) (Variables|Default|Getw)) (Variables|Default|GetMovement%20Speed))) (* _x (* (* (Math|Conversions|ToFloat(Boolean) (Variables|Default|Gets)) (Variables|Default|GetMovement%20Speed)) -1.0))) (+ (+ (* _y (* (Math|Conversions|ToFloat(Boolean) (Variables|Default|Getd)) (Variables|Default|GetMovement%20Speed))) (* _y (* (* (Math|Conversions|ToFloat(Boolean) (Variables|Default|Geta)) (Variables|Default|GetMovement%20Speed)) -1.0))) (+ (* _z (* (Math|Conversions|ToFloat(Boolean) (Variables|Default|Getq)) (Variables|Default|GetMovement%20Speed))) (* _z (* (* (Math|Conversions|ToFloat(Boolean) (Variables|Default|Gete)) (Variables|Default|GetMovement%20Speed)) -1.0)))))))

(event MiddleMouseButton
  (params (Key :type Key%20Structure))
  (bind _brushhitlocationwhenmeasurestarted (Variables|Default|GetBrushHitLocationWhenMeasureStarted))
  (bind _brush_hit_location (Variables|Default|GetBrush%20Hit%20Location))
  (bind _returnvalue (- _brush_hit_location _brushhitlocationwhenmeasurestarted))
  (bind _paintcloudsbp (Variables|Default|GetPaintCloudsBP)))

(event MouseWheelUp
  (params (Key :type Key%20Structure)))

(event MouseWheelDown
  (params (Key :type Key%20Structure)))

(event MouseWheelUp
  (params (Key :type Key%20Structure)))

(event MouseWheelDown
  (params (Key :type Key%20Structure)))

(event %31
  (params (Key :type Key%20Structure)))

(event Num+
  (params (Key :type Key%20Structure)))

(event %32
  (params (Key :type Key%20Structure)))

(event LeftShift
  (params (Key :type Key%20Structure)))

(event %33
  (params (Key :type Key%20Structure)))

(event RightShift
  (params (Key :type Key%20Structure)))

(event W
  (params (Key :type Key%20Structure)))

(event %34
  (params (Key :type Key%20Structure)))

(event S
  (params (Key :type Key%20Structure)))

(event A
  (params (Key :type Key%20Structure)))

(event D
  (params (Key :type Key%20Structure)))

(event E
  (params (Key :type Key%20Structure)))

(event Q
  (params (Key :type Key%20Structure)))

(fn ConvertBrushSize
  (params (Local%20Size :type Float%20%28double-precision%29))
  (Variables|Default|GetPaintCloudsBP
    (:then
      (return (* Local%20Size (/ (* (Variables|/Volumetrics/Tools/CloudCompositing/Blueprints/BP_PaintClouds.BP_PaintClouds_C|GetCloud%20Scale%20Km (Variables|Default|GetPaintCloudsBP)) 1e+05) 1e+01))))
    (:else)))

(fn InitPaintCloudsBP
  (params (BP_PaintClouds :type BP%20Paint%20Clouds%20Object%20Reference))
  (Utilities|IsValid BP_PaintClouds
    (:Is%20Valid
      (if (== BP_PaintClouds (Variables|Default|GetPaintCloudsBP))
        (Development|PrintString "PaintClouds BP initiated multiple times!" true true "(R=1.000000,G=0.000000,B=0.000000,A=1.000000)" 5.0)
        (else
          (Variables|Default|GetPaintCloudsBP
            (:then
              (Development|PrintString "PaintClouds BP initiated when there already is one" true true "(R=1.000000,G=0.000000,B=0.000000,A=1.000000)" 5.0))
            (:else
              (Variables|Default|SetPaintCloudsBP BP_PaintClouds))))))
    (:Is%20Not%20Valid
      (Development|PrintString "Invalid BP_PaintClouds reference." true true "(R=1.000000,G=0.000000,B=0.000000,A=1.000000)" 5.0))))

(fn RayTracetoCloudLayer
  (bind _returnvalue (Game|GetPlayerController 0))
  (bind (_worldlocation _worlddirection) (Game|Player|ConvertMouseLocationToWorldSpace _returnvalue))
  (bind (_t _intersection) (Math|Intersection|LinePlaneIntersection(Origin&Normal) (LocalVariables|GetTraceStart) (LocalVariables|GetTraceEnd) (Math|Vector|MakeVector 0.0 0.0 (* (Variables|/Volumetrics/Tools/CloudCompositing/Blueprints/BP_PaintClouds.BP_PaintClouds_C|GetElevation%20to%20Paint%20at%20Km (Variables|Default|GetPaintCloudsBP)) 1e+05)) "0, 0,-1.000000"))
  (bind _traceend (LocalVariables|GetTraceEnd))
  (Utilities|IsValid (Variables|Default|GetPaintCloudsBP)
    (:Is%20Valid
      (Variables|Default|SetPrevious%20Brush%20Hit%20Location (Variables|Default|GetBrush%20Hit%20Location))
      (LocalVariables|SetTraceStart _worldlocation)
      (LocalVariables|SetTraceEnd (+ _worldlocation (* _worlddirection 1e+08)))
      (Utilities|FlowControl|Sequence
        (:then_0
          (if (or (Utilities|Enum|Equal(Enum) (Variables|/Volumetrics/Tools/CloudCompositing/Blueprints/BP_PaintClouds.BP_PaintClouds_C|GetBrushTraceMode (Variables|Default|GetPaintCloudsBP)) "NewEnumerator0") (Utilities|Enum|Equal(Enum) (Variables|/Volumetrics/Tools/CloudCompositing/Blueprints/BP_PaintClouds.BP_PaintClouds_C|GetBrushTraceMode (Variables|Default|GetPaintCloudsBP)) "NewEnumerator2"))
            (LocalVariables|SetZPlaneHitLocation _intersection)))
        (:then_1
          (if (or (Utilities|Enum|Equal(Enum) (Variables|/Volumetrics/Tools/CloudCompositing/Blueprints/BP_PaintClouds.BP_PaintClouds_C|GetBrushTraceMode (Variables|Default|GetPaintCloudsBP)) "NewEnumerator1") (Utilities|Enum|Equal(Enum) (Variables|/Volumetrics/Tools/CloudCompositing/Blueprints/BP_PaintClouds.BP_PaintClouds_C|GetBrushTraceMode (Variables|Default|GetPaintCloudsBP)) "NewEnumerator2"))
            (bind _outhit (Collision|LineTraceByChannel :Start (LocalVariables|GetTraceStart) :End _traceend :TraceChannel "TraceTypeQuery4" :bTraceComplex true :DrawDebugType "None" :bIgnoreSelf false :TraceColor "(R=1.000000,G=0.000000,B=0.000000,A=1.000000)" :TraceHitColor "(R=0.000000,G=1.000000,B=0.000000,A=1.000000)" :DrawTime 5.0))
            (LocalVariables|SetVolumetricPaintingChannelTraceLocation (Math|Vector|SelectVector (.Location (Collision|BreakHitResult _outhit)) _traceend (Collision|BreakHitResult _outhit)))))
        (:then_2
          (switch Utilities|FlowControl|Switch|SwitchonVolumetricPaintingBrushTraceMode (Variables|/Volumetrics/Tools/CloudCompositing/Blueprints/BP_PaintClouds.BP_PaintClouds_C|GetBrushTraceMode (Variables|Default|GetPaintCloudsBP))
            (:NewEnumerator0
              (Variables|Default|SetBrush%20Hit%20Location (LocalVariables|GetZPlaneHitLocation)))
            (:NewEnumerator1
              (Variables|Default|SetBrush%20Hit%20Location (LocalVariables|GetVolumetricPaintingChannelTraceLocation)))
            (:NewEnumerator2
              (Variables|Default|SetBrush%20Hit%20Location (Math|Vector|SelectVector (LocalVariables|GetZPlaneHitLocation) (LocalVariables|GetVolumetricPaintingChannelTraceLocation) (< (Math|Vector|VectorLength (- (LocalVariables|GetZPlaneHitLocation) (LocalVariables|GetTraceStart))) (Math|Vector|VectorLength (- (LocalVariables|GetVolumetricPaintingChannelTraceLocation) (LocalVariables|GetTraceStart)))))))))))))

(fn SetPaintBlendMode
  (params (PaintBlendMode :type CloudPaintBlendMode%20Enum :default "NewEnumerator0"))
  (Variables|Default|SetPaintBlendMode PaintBlendMode))

(fn UpdateBrushPreview
  (bind _brushpreview (Variables|Default|GetBrushPreview))
  (Transformation|SetWorldLocation _brushpreview (Variables|Default|GetBrush%20Hit%20Location) false false)
  (Rendering|SetVisibility _brushpreview (not (Math|Vector|Equal(Vector) (Variables|Default|GetBrush%20Hit%20Location) "0, 0, 0" 0.0001)) false))

(fn ConstructionScript)
