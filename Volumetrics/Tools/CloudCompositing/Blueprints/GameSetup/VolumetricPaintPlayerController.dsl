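; Player controller Blueprint for the volumetric paint setup in the cloud compositing tools, serving to announce pawn possession.
; On possess, it calls the OnPossessDelegate event dispatcher with the possessed pawn so that any bound listeners can respond.
; BeginPlay, Tick and the construction script are empty, so this relay is its only graph logic.
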(event EventBeginPlay)

(event EventTick
  (params (DeltaSeconds :type Float%20%28single-precision%29)))

(event EventOnPossess
  (params (PossessedPawn :type Pawn%20Object%20Reference))
  (Default|CallOnPossessDelegate PossessedPawn))

(fn ConstructionScript)
